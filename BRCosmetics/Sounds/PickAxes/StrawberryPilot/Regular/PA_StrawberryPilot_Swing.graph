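# Sound cue for the StrawberryPilot pickaxe swing cosmetic.
# Each quality level mixes random pickaxe-specific swings under close attenuation with shared fire axe distant swings under medium attenuation, with modulators varying the distant layer.
# Variant pools shrink across levels; the last two share one mix whose single distant wave gets a wider pitch range.
# It primes on load and overrides concurrency to limit voices per owner, stopping the oldest.
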SoundCue /BRCosmetics/Sounds/PickAxes/StrawberryPilot/Regular/PA_StrawberryPilot_Swing
root=QualityLevel_0
nodes [
 {id=Attenuation_0, children [Random_7], properties {attenuationSettings=/Game/Sounds/Attenuation/Fort_Weapon_Pickaxe_Close_Att}}
 {id=Attenuation_1, children [Random_5], properties {attenuationSettings=/Game/Sounds/Attenuation/Fort_Weapon_Pickaxe_Medium_Att}}
 {id=Attenuation_2, children [Random_8], properties {attenuationSettings=&Fort_Weapon_Pickaxe_Close_Att}}
 {id=Attenuation_3, children [Random_6], properties {attenuationSettings=&Fort_Weapon_Pickaxe_Medium_Att}}
 {id=Attenuation_4, children [Random_9], properties {attenuationSettings=&Fort_Weapon_Pickaxe_Close_Att}}
 {id=Attenuation_5, children [Modulator_3], properties {attenuationSettings=&Fort_Weapon_Pickaxe_Medium_Att}}
 {id=Mixer_0, children [Attenuation_0, Modulator_4], properties {inputVolume [1,1]}}
 {id=Mixer_1, children [Attenuation_2, Modulator_0], properties {inputVolume [1,1]}}
 {id=Mixer_2, children [Attenuation_4, Attenuation_5], properties {inputVolume [1,1]}}
 {id=Modulator_0, children [Attenuation_3]}
 {id=Modulator_3, children [WavePlayer_30], properties {pitchMax=1.2, pitchMin=0.9}}
 {id=Modulator_4, children [Attenuation_1]}
 {id=QualityLevel_0, children [Mixer_0,Mixer_1,Mixer_2,Mixer_2]}
 {
  id=Random_5
  children [WavePlayer_18, WavePlayer_19, WavePlayer_20, WavePlayer_24, WavePlayer_25, WavePlayer_26]
  properties {weights [1,1,1,1,1,1]}
 }
 {id=Random_6, children [WavePlayer_28, WavePlayer_27, WavePlayer_29], properties {weights [1,1,1]}}
 {id=Random_7, children [WavePlayer_33, WavePlayer_34, WavePlayer_35, WavePlayer_36], properties {weights [1,1,1,1]}}
 {id=Random_8, children [WavePlayer_42, WavePlayer_43, WavePlayer_44], properties {weights [1,1,1]}}
 {id=Random_9, children [WavePlayer_45, WavePlayer_46], properties {weights [1,1]}}
 {id=WavePlayer_18, properties {soundWaveAssetPtr=/Game/Athena/Sounds/Weapons/Fire_Axe_Swing_Distant_04}}
 {id=WavePlayer_19, properties {soundWaveAssetPtr=/Game/Athena/Sounds/Weapons/Fire_Axe_Swing_Distant_01}}
 {id=WavePlayer_20, properties {soundWaveAssetPtr=/Game/Athena/Sounds/Weapons/Fire_Axe_Swing_Distant_02}}
 {id=WavePlayer_24, properties {soundWaveAssetPtr=/Game/Athena/Sounds/Weapons/Fire_Axe_Swing_Distant_03}}
 {id=WavePlayer_25, properties {soundWaveAssetPtr=/Game/Athena/Sounds/Weapons/Fire_Axe_Swing_Distant_05}}
 {id=WavePlayer_26, properties {soundWaveAssetPtr=/Game/Athena/Sounds/Weapons/Fire_Axe_Swing_Distant_06}}
 {id=WavePlayer_27, properties {soundWaveAssetPtr=&Fire_Axe_Swing_Distant_03}}
 {id=WavePlayer_28, properties {soundWaveAssetPtr=&Fire_Axe_Swing_Distant_04}}
 {id=WavePlayer_29, properties {soundWaveAssetPtr=&Fire_Axe_Swing_Distant_06}}
 {id=WavePlayer_30, properties {soundWaveAssetPtr=&Fire_Axe_Swing_Distant_06}}
 {id=WavePlayer_33, properties {soundWaveAssetPtr=&/PA_StrawberryPilot_Swing_03}}
 {id=WavePlayer_34, properties {soundWaveAssetPtr=&/PA_StrawberryPilot_Swing_04}}
 {id=WavePlayer_35, properties {soundWaveAssetPtr=&/PA_StrawberryPilot_Swing_05}}
 {id=WavePlayer_36, properties {soundWaveAssetPtr=&/PA_StrawberryPilot_Swing_06}}
 {id=WavePlayer_42, properties {soundWaveAssetPtr=&PA_StrawberryPilot_Swing_03}}
 {id=WavePlayer_43, properties {soundWaveAssetPtr=&PA_StrawberryPilot_Swing_04}}
 {id=WavePlayer_44, properties {soundWaveAssetPtr=&PA_StrawberryPilot_Swing_05}}
 {id=WavePlayer_45, properties {soundWaveAssetPtr=&PA_StrawberryPilot_Swing_03}}
 {id=WavePlayer_46, properties {soundWaveAssetPtr=&PA_StrawberryPilot_Swing_04}}
]
properties {
 audioPropertiesSheet=/BRCosmetics/Sounds/Settings/PropertySheets/Pickaxe/APS_Cosmetic_Pickaxe_Swing
 bOverrideConcurrency=true
 bPrimeOnLoad=true
 maxDistance=3000
 soundClassObject=&Fort_WeaponMelee_PSM
 volumeMultiplier=0.4
 concurrencyOverrides {bLimitToOwner=true, maxCount=2, resolutionRule=StopOldest}
 concurrencySet [&Player_Melee_Weapon_Conc]
}
